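# Physics asset for the F_MED_WaningCacti female body cosmetic, giving its dynamic belts physics bodies alongside capsules for the lower spine, pelvis, legs and feet.
# Four six-link belt chains (front/back, left/right) and two short side anchor chains hang from zero-size root bodies with free angular constraints to the pelvis.
# Links have limited swing and twist.
# Each belt's last link joins its side's anchor tip with free rotation, closing loops.
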PhysicsAsset /BRCosmetics/Characters/Player/Female/Medium/Bodies/F_MED_WaningCacti/Meshes/F_MED_WaningCacti_Physics
bodies {
 calf_l [{shapeType=Capsule, length=36.6814, radius=7.2619, center=<-24.0398 -2.1936 0>, rotation=<rot 90 0 0>}]
 calf_r [{shapeType=Capsule, length=36.6814, radius=7.2619, center=<24.0398 2.1936 0>, rotation=<rot 90 180 0>}]
 dyn_belt_anchor_l_01 [{shapeType=Capsule, length=7.3839, radius=4, shapeName=dyn_belt_anchor_l_01_capsule, center=<7.2607 0 0>, rotation=<rot 90 0 0>}]
 dyn_belt_anchor_l_02 [
  {shapeType=Capsule, length=0, radius=0, shapeName=dyn_belt_anchor_l_02_capsule}
  {shapeType=Box, extentX=20.7169, extentY=15.8793, extentZ=6.2427, shapeName=dyn_belt_anchor_l_02_box, center=<6.5159 0 2.2037>}
 ]
 dyn_belt_anchor_l_master [{shapeType=Capsule, length=0, radius=0, shapeName=dyn_belt_anchor_l_master_capsule}]
 dyn_belt_anchor_r_01 [{shapeType=Capsule, length=6.4443, radius=3.928, shapeName=dyn_belt_anchor_r_01_capsule, center=<6.5957 0 0>, rotation=<rot 90 0 0>}]
 dyn_belt_anchor_r_02 [{shapeType=Capsule, length=7.1073, radius=4, shapeName=dyn_belt_anchor_r_02_capsule, center=<1.4614 0 1.3164>, rotation=<rot 0 0 90>}]
 dyn_belt_anchor_r_master [{shapeType=Capsule, length=0, radius=0, shapeName=dyn_belt_anchor_r_master_capsule}]
 dyn_belt_bk_l_01 [{shapeType=Capsule, length=0.505, radius=0.505, shapeName=dyn_belt_bk_l_01_capsule, center=<0 0 0.0576>, rotation=<rot 0 -90 0>}]
 dyn_belt_bk_l_02 [{
   shapeType=Capsule
   length=3.1477
   radius=1.655
   shapeName=dyn_belt_bk_l_02_capsule
   center=<1.5565 -0.0204 -0.3339>
   rotation=<rot 0 -90.7479 -80.7373>
  }]
 dyn_belt_bk_l_03 [{
   shapeType=Capsule
   length=3.4366
   radius=1.5617
   shapeName=dyn_belt_bk_l_03_capsule
   center=<2.2932 0.0573 0.0452>
   rotation=<rot 0 -87.022 -84.0416>
  }]
 dyn_belt_bk_l_04 [{
   shapeType=Capsule
   length=4.9856
   radius=1.8934
   shapeName=dyn_belt_bk_l_04_capsule
   center=<2.3584 0.225 -0.1856>
   rotation=<rot -2.236 -77.1042 -78.0298>
  }]
 dyn_belt_bk_l_05 [{shapeType=Capsule, length=4.6233, radius=1.8782, shapeName=dyn_belt_bk_l_05_capsule, center=<1.959 0 0>, rotation=<rot 90 0 0>}]
 dyn_belt_bk_l_06 [{
   shapeType=Capsule
   length=5.0318
   radius=1.7948
   shapeName=dyn_belt_bk_l_06_capsule
   center=<-0.0228 -0.0168 0.8132>
   rotation=<rot 0 87.1956 -76.6811>
  }]
 dyn_belt_bk_l_base [{shapeType=Capsule, length=0, radius=0, shapeName=dyn_belt_bk_l_base_capsule}]
 dyn_belt_bk_r_01 [{shapeType=Capsule, length=1.4383, radius=0.9329, shapeName=dyn_belt_bk_r_01_capsule, center=<1.0613 0 0>, rotation=<rot 90 0 0>}]
 dyn_belt_bk_r_02 [{
   shapeType=Capsule
   length=3.4105
   radius=1.0803
   shapeName=dyn_belt_bk_r_02_capsule
   center=<2.146 0.0913 -0.1985>
   rotation=<rot 0 92.6069 -86.7848>
  }]
 dyn_belt_bk_r_03 [{
   shapeType=Capsule
   length=4.0595
   radius=1.9415
   shapeName=dyn_belt_bk_r_03_capsule
   center=<2.4292 -0.1421 0.2803>
   rotation=<rot 0 86.902 -86.3569>
  }]
 dyn_belt_bk_r_04 [{
   shapeType=Capsule
   length=4.9253
   radius=2.5109
   shapeName=dyn_belt_bk_r_04_capsule
   center=<2.4447 -0.0213 0.3036>
   rotation=<rot 0 -86.2395 -83.771>
  }]
 dyn_belt_bk_r_05 [{
   shapeType=Capsule
   length=5.0882
   radius=2.4805
   shapeName=dyn_belt_bk_r_05_capsule
   center=<2.0784 0.1214 -0.1768>
   rotation=<rot 0 -82.3133 -82.7825>
  }]
 dyn_belt_bk_r_06 [{
   shapeType=Capsule
   length=5.3332
   radius=2.1909
   shapeName=dyn_belt_bk_r_06_capsule
   center=<1.19 0.1437 0.2252>
   rotation=<rot 0 86.0694 -79.5229>
  }]
 dyn_belt_bk_r_base [{shapeType=Capsule, length=0, radius=0, shapeName=dyn_belt_bk_r_base_capsule}]
 dyn_belt_frnt_l_01 [{
   shapeType=Capsule
   length=4.1957
   radius=1.2045
   shapeName=dyn_belt_frnt_l_01_capsule
   center=<2.4404 -0.1143 0.2113>
   rotation=<rot 0 -84.185 -81.4404>
  }]
 dyn_belt_frnt_l_02 [{
   shapeType=Capsule
   length=4.6301
   radius=1.5252
   shapeName=dyn_belt_frnt_l_02_capsule
   center=<2.6197 -0.0585 -0.1225>
   rotation=<rot 0 91.401 -87.292>
  }]
 dyn_belt_frnt_l_03 [{
   shapeType=Capsule
   length=5.0115
   radius=1.6527
   shapeName=dyn_belt_frnt_l_03_capsule
   center=<1.7985 -0.1512 -0.2925>
   rotation=<rot 0.0001 96.5379 -88.9317>
  }]
 dyn_belt_frnt_l_04 [{
   shapeType=Capsule
   length=5.1571
   radius=1.8863
   shapeName=dyn_belt_frnt_l_04_capsule
   center=<2.2963 -0.1018 0.2408>
   rotation=<rot 0 89.3558 -87.7286>
  }]
 dyn_belt_frnt_l_05 [{
   shapeType=Capsule
   length=4.8029
   radius=1.6041
   shapeName=dyn_belt_frnt_l_05_capsule
   center=<1.6289 -0.0753 0.0545>
   rotation=<rot 0 -110.0721 -72.9863>
  }]
 dyn_belt_frnt_l_06 [{
   shapeType=Capsule
   length=5.1888
   radius=1.6995
   shapeName=dyn_belt_frnt_l_06_capsule
   center=<-0.4905 -0.1107 0.3053>
   rotation=<rot 0 87.3733 -80.9033>
  }]
 dyn_belt_frnt_l_base [{shapeType=Capsule, length=0, radius=0, shapeName=dyn_belt_frnt_l_base_capsule}]
 dyn_belt_frnt_r_01 [{
   shapeType=Capsule
   length=2.6242
   radius=1.5019
   shapeName=dyn_belt_frnt_r_01_capsule
   center=<2.4259 -0.12 -0.2296>
   rotation=<rot 0 95.2184 -78.7038>
  }]
 dyn_belt_frnt_r_02 [{shapeType=Capsule, length=3.2127, radius=1.7262, shapeName=dyn_belt_frnt_r_02_capsule, center=<1.9141 0 0>, rotation=<rot 90 0 0>}]
 dyn_belt_frnt_r_03 [{
   shapeType=Capsule
   length=4.9299
   radius=1.8982
   shapeName=dyn_belt_frnt_r_03_capsule
   center=<2.1265 -0.1423 0.0881>
   rotation=<rot 0 89.5737 -81.6693>
  }]
 dyn_belt_frnt_r_04 [{
   shapeType=Capsule
   length=4.7555
   radius=2.194
   shapeName=dyn_belt_frnt_r_04_capsule
   center=<2.5052 -0.1378 0.2921>
   rotation=<rot 0 87.0859 -89.6379>
  }]
 dyn_belt_frnt_r_05 [{
   shapeType=Capsule
   length=4.7971
   radius=1.4689
   shapeName=dyn_belt_frnt_r_05_capsule
   center=<2.4787 0.0986 0.1788>
   rotation=<rot 0 90.433 -91.144>
  }]
 dyn_belt_frnt_r_06 [{
   shapeType=Capsule
   length=4.8367
   radius=1.9908
   shapeName=dyn_belt_frnt_r_06_capsule
   center=<-0.7958 0.0752 0.0161>
   rotation=<rot 0 -90.8656 -80.1458>
  }]
 dyn_belt_frnt_r_base [{shapeType=Capsule, length=0, radius=0, shapeName=dyn_belt_frnt_r_base_capsule}]
 foot_l [{shapeType=Capsule, length=13.5788, radius=9.7282, center=<-3.7356 8.891 0>, rotation=<rot 90 -84.2579 0>}]
 foot_r [{shapeType=Capsule, length=13.5788, radius=9.7282, center=<3.7356 -8.891 0>, rotation=<rot 90 95.7421 0>}]
 pelvis [{shapeType=Capsule, length=8.5212, radius=10.5135, center=<-6.5439 -0.7223 0>}]
 spine_01 [{shapeType=Capsule, length=14.846, radius=7.9737, shapeName=spine_01_capsule, center=<-4.3095 -0.2872 0>}]
 thigh_l [{shapeType=Capsule, length=40.4928, radius=7.2219, center=<-24.0398 0 0.789>, rotation=<rot 90 0 0>}]
 thigh_r [{shapeType=Capsule, length=39.4733, radius=8.2414, center=<24.0398 1.0161 -1.3062>, rotation=<rot 90 180 0>}]
}
constraints [
 {bone1Name=thigh_l, bone2Name=pelvis, swing1LimitDegrees=45, swing2LimitDegrees=45, twistLimitDegrees=45}
 {bone1Name=calf_l, bone2Name=thigh_l, swing1LimitDegrees=45, swing2LimitDegrees=45, twistLimitDegrees=45}
 {bone1Name=foot_l, bone2Name=calf_l, swing1LimitDegrees=45, swing2LimitDegrees=45, twistLimitDegrees=45}
 {bone1Name=thigh_r, bone2Name=pelvis, swing1LimitDegrees=45, swing2LimitDegrees=45, twistLimitDegrees=45}
 {bone1Name=calf_r, bone2Name=thigh_r, swing1LimitDegrees=45, swing2LimitDegrees=45, twistLimitDegrees=45}
 {bone1Name=foot_r, bone2Name=calf_r, swing1LimitDegrees=45, swing2LimitDegrees=45, twistLimitDegrees=45}
 {
  bone1Name=spine_01
  bone2Name=pelvis
  swing1LimitDegrees=45
  swing1Motion=Free
  swing2LimitDegrees=45
  swing2Motion=Free
  twistLimitDegrees=45
  twistMotion=Free
 }
 {
  bone1Name=dyn_belt_anchor_l_master
  bone2Name=pelvis
  swing1LimitDegrees=45
  swing1Motion=Free
  swing2LimitDegrees=45
  swing2Motion=Free
  twistLimitDegrees=45
  twistMotion=Free
 }
 {bone1Name=dyn_belt_anchor_l_01, bone2Name=dyn_belt_anchor_l_master, swing1LimitDegrees=20, swing2LimitDegrees=45, twistLimitDegrees=45}
 {bone1Name=dyn_belt_anchor_l_02, bone2Name=dyn_belt_anchor_l_01, swing1LimitDegrees=25, swing2LimitDegrees=20, twistLimitDegrees=20}
 {
  bone1Name=dyn_belt_anchor_r_master
  bone2Name=pelvis
  swing1LimitDegrees=45
  swing1Motion=Free
  swing2LimitDegrees=45
  swing2Motion=Free
  twistLimitDegrees=45
  twistMotion=Free
 }
 {bone1Name=dyn_belt_anchor_r_01, bone2Name=dyn_belt_anchor_r_master, swing1LimitDegrees=25, swing2LimitDegrees=25, twistLimitDegrees=45}
 {bone1Name=dyn_belt_anchor_r_02, bone2Name=dyn_belt_anchor_r_01, swing1LimitDegrees=45, swing2LimitDegrees=45, twistLimitDegrees=45}
 {
  bone1Name=dyn_belt_bk_l_base
  bone2Name=pelvis
  swing1LimitDegrees=45
  swing1Motion=Free
  swing2LimitDegrees=45
  swing2Motion=Free
  twistLimitDegrees=45
  twistMotion=Free
 }
 {bone1Name=dyn_belt_bk_l_01, bone2Name=dyn_belt_bk_l_base, swing1LimitDegrees=10, swing2LimitDegrees=10, twistLimitDegrees=0}
 {bone1Name=dyn_belt_bk_l_02, bone2Name=dyn_belt_bk_l_01, swing1LimitDegrees=20, swing2LimitDegrees=20, twistLimitDegrees=5}
 {bone1Name=dyn_belt_bk_l_03, bone2Name=dyn_belt_bk_l_02, swing1LimitDegrees=25, swing2LimitDegrees=25, twistLimitDegrees=15}
 {bone1Name=dyn_belt_bk_l_04, bone2Name=dyn_belt_bk_l_03, swing1LimitDegrees=25, swing2LimitDegrees=25, twistLimitDegrees=0}
 {bone1Name=dyn_belt_bk_l_05, bone2Name=dyn_belt_bk_l_04, swing1LimitDegrees=25, swing2LimitDegrees=25, twistLimitDegrees=0}
 {bone1Name=dyn_belt_bk_l_06, bone2Name=dyn_belt_bk_l_05, swing1LimitDegrees=25, swing2LimitDegrees=25, twistLimitDegrees=15}
 {
  bone1Name=dyn_belt_bk_r_base
  bone2Name=pelvis
  swing1LimitDegrees=45
  swing1Motion=Free
  swing2LimitDegrees=45
  swing2Motion=Free
  twistLimitDegrees=45
  twistMotion=Free
 }
 {bone1Name=dyn_belt_bk_r_01, bone2Name=dyn_belt_bk_r_base, swing1LimitDegrees=5, swing2LimitDegrees=5, twistLimitDegrees=5}
 {bone1Name=dyn_belt_bk_r_02, bone2Name=dyn_belt_bk_r_01, swing1LimitDegrees=25, swing2LimitDegrees=25, twistLimitDegrees=5}
 {bone1Name=dyn_belt_bk_r_03, bone2Name=dyn_belt_bk_r_02, swing1LimitDegrees=25, swing2LimitDegrees=25, twistLimitDegrees=5}
 {bone1Name=dyn_belt_bk_r_04, bone2Name=dyn_belt_bk_r_03, swing1LimitDegrees=25, swing2LimitDegrees=25, twistLimitDegrees=5}
 {bone1Name=dyn_belt_bk_r_05, bone2Name=dyn_belt_bk_r_04, swing1LimitDegrees=25, swing2LimitDegrees=25, twistLimitDegrees=10}
 {bone1Name=dyn_belt_bk_r_06, bone2Name=dyn_belt_bk_r_05, swing1LimitDegrees=25, swing2LimitDegrees=25, twistLimitDegrees=15}
 {
  bone1Name=dyn_belt_frnt_l_base
  bone2Name=pelvis
  swing1LimitDegrees=45
  swing1Motion=Free
  swing2LimitDegrees=45
  swing2Motion=Free
  twistLimitDegrees=45
  twistMotion=Free
 }
 {bone1Name=dyn_belt_frnt_l_01, bone2Name=dyn_belt_frnt_l_base, swing1LimitDegrees=45, swing2LimitDegrees=45, twistLimitDegrees=0}
 {bone1Name=dyn_belt_frnt_l_02, bone2Name=dyn_belt_frnt_l_01, swing1LimitDegrees=25, swing2LimitDegrees=25, twistLimitDegrees=0}
 {bone1Name=dyn_belt_frnt_l_03, bone2Name=dyn_belt_frnt_l_02, swing1LimitDegrees=25, swing2LimitDegrees=25, twistLimitDegrees=5}
 {bone1Name=dyn_belt_frnt_l_04, bone2Name=dyn_belt_frnt_l_03, swing1LimitDegrees=25, swing2LimitDegrees=25, twistLimitDegrees=0}
 {bone1Name=dyn_belt_frnt_l_05, bone2Name=dyn_belt_frnt_l_04, swing1LimitDegrees=25, swing2LimitDegrees=25, twistLimitDegrees=0}
 {bone1Name=dyn_belt_frnt_l_06, bone2Name=dyn_belt_frnt_l_05, swing1LimitDegrees=25, swing2LimitDegrees=25, twistLimitDegrees=0}
 {
  bone1Name=dyn_belt_frnt_r_base
  bone2Name=pelvis
  swing1LimitDegrees=45
  swing1Motion=Free
  swing2LimitDegrees=45
  swing2Motion=Free
  twistLimitDegrees=45
  twistMotion=Free
 }
 {bone1Name=dyn_belt_frnt_r_01, bone2Name=dyn_belt_frnt_r_base, swing1LimitDegrees=45, swing2LimitDegrees=45, twistLimitDegrees=0}
 {bone1Name=dyn_belt_frnt_r_02, bone2Name=dyn_belt_frnt_r_01, swing1LimitDegrees=25, swing2LimitDegrees=25, twistLimitDegrees=0}
 {bone1Name=dyn_belt_frnt_r_03, bone2Name=dyn_belt_frnt_r_02, swing1LimitDegrees=25, swing2LimitDegrees=25, twistLimitDegrees=0}
 {bone1Name=dyn_belt_frnt_r_04, bone2Name=dyn_belt_frnt_r_03, swing1LimitDegrees=25, swing2LimitDegrees=25, twistLimitDegrees=0}
 {bone1Name=dyn_belt_frnt_r_05, bone2Name=dyn_belt_frnt_r_04, swing1LimitDegrees=25, swing2LimitDegrees=25, twistLimitDegrees=0}
 {bone1Name=dyn_belt_frnt_r_06, bone2Name=dyn_belt_frnt_r_05, swing1LimitDegrees=25, swing2LimitDegrees=25, twistLimitDegrees=0}
 {
  bone1Name=dyn_belt_frnt_r_06
  bone2Name=dyn_belt_anchor_r_02
  swing1LimitDegrees=45
  swing1Motion=Free
  swing2LimitDegrees=45
  swing2Motion=Free
  twistLimitDegrees=45
  twistMotion=Free
 }
 {
  bone1Name=dyn_belt_bk_r_06
  bone2Name=dyn_belt_anchor_r_02
  swing1LimitDegrees=45
  swing1Motion=Free
  swing2LimitDegrees=45
  swing2Motion=Free
  twistLimitDegrees=45
  twistMotion=Free
 }
 {
  bone1Name=dyn_belt_bk_l_06
  bone2Name=dyn_belt_anchor_l_02
  swing1LimitDegrees=45
  swing1Motion=Free
  swing2LimitDegrees=45
  swing2Motion=Free
  twistLimitDegrees=45
  twistMotion=Free
 }
 {
  bone1Name=dyn_belt_frnt_l_06
  bone2Name=dyn_belt_anchor_l_02
  swing1LimitDegrees=45
  swing1Motion=Free
  swing2LimitDegrees=45
  swing2Motion=Free
  twistLimitDegrees=45
  twistMotion=Free
 }
]
