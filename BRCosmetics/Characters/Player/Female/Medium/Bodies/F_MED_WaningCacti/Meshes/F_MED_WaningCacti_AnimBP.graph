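# Post-process style Anim Blueprint for the female WaningCacti body cosmetic, adding secondary motion to its clothing and accessories.
# It copies the parent mesh pose, then runs an emote-aware Control Rig, a pelvis-space RigidBody simulation and a small belt-anchor offset.
# Sprung, angle-limited AnimDynamics chains, linked by cached poses, move necklace, zippos, leg laces and skirt.
# Weighted constraints then reposition zippo bases from body and skirt bones before bag dynamics produce the output.
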AnimBlueprint /BRCosmetics/Characters/Player/Female/Medium/Bodies/F_MED_WaningCacti/Meshes/F_MED_WaningCacti_AnimBP
target_skeleton=&/F_MED_WaningCacti_Skeleton
graphs [
 AnimGraph {nodes [
   n1 = Anim Dynamics - Bone: dyn_bag_01(ComponentPose=Local To Component(LocalPose=Use cached pose 'toAnimDyn4'()), Alpha=1, ExternalForce="0,0,0") {
    angularDampingOverride=1
    angularSpringConstant=500
    bAngularSpring=true
    bOverrideAngularDamping=true
    bOverrideLinearDamping=true
    bUseGravityOverride=true
    linearDampingOverride=1
    simulationSpace=RootRelative
    boundBone {boneName=dyn_bag_01}
    componentAppliedLinearAccClamp=<2000 2000 2000>
    componentLinearAccScale=<1 1 1>
    componentLinearVelScale=<1 1 1>
    gravityOverride {z=-980}
    physicsBodyDefinitions [{
      boundBone {boneName=dyn_bag_01}
      boxExtents=<7 7 7>
      constraintSetup {angularLimitsMax=<5 0 10>, angularLimitsMin=<-5 -20 -20>, angularTarget=<1 0 0>}
      localJointOffset=<3 0 0>
     }]
   }
   n2 = Anim Dynamics - Bone: dyn_bag_03(ComponentPose=@n1, Alpha=1, ExternalForce="0,0,0") {
    angularDampingOverride=1
    angularSpringConstant=500
    bAngularSpring=true
    bOverrideAngularDamping=true
    bOverrideLinearDamping=true
    bUseGravityOverride=true
    linearDampingOverride=1
    simulationSpace=RootRelative
    boundBone {boneName=dyn_bag_03}
    componentAppliedLinearAccClamp=<2000 2000 2000>
    componentLinearAccScale=<1 1 1>
    componentLinearVelScale=<1 1 1>
    gravityOverride {z=-980}
    physicsBodyDefinitions [{
      boundBone {boneName=dyn_bag_03}
      boxExtents=<7 7 7>
      constraintSetup {angularLimitsMax=<5 0 10>, angularLimitsMin=<-5 -20 -20>, angularTarget=<1 0 0>}
      localJointOffset=<3 0 0>
     }]
   }
   n3 = Anim Dynamics - Bone: dyn_bag_05(ComponentPose=@n2, Alpha=1, ExternalForce="0,0,0") {
    angularDampingOverride=1
    angularSpringConstant=500
    bAngularSpring=true
    bOverrideAngularDamping=true
    bOverrideLinearDamping=true
    bUseGravityOverride=true
    linearDampingOverride=1
    simulationSpace=RootRelative
    boundBone {boneName=dyn_bag_05}
    componentAppliedLinearAccClamp=<2000 2000 2000>
    componentLinearAccScale=<1 1 1>
    componentLinearVelScale=<1 1 1>
    gravityOverride {z=-980}
    physicsBodyDefinitions [{
      boundBone {boneName=dyn_bag_05}
      boxExtents=<7 7 7>
      constraintSetup {angularLimitsMax=<5 0 10>, angularLimitsMin=<-5 -20 -20>, angularTarget=<1 0 0>}
      localJointOffset=<3 0 0>
     }]
   }
   n4 = Anim Dynamics - Bone: dyn_bag_02(ComponentPose=@n3, Alpha=1, ExternalForce="0,0,0") {
    angularDampingOverride=1
    angularSpringConstant=500
    bAngularSpring=true
    bOverrideAngularDamping=true
    bOverrideLinearDamping=true
    bUseGravityOverride=true
    linearDampingOverride=1
    simulationSpace=RootRelative
    boundBone {boneName=dyn_bag_02}
    componentAppliedLinearAccClamp=<2000 2000 2000>
    componentLinearAccScale=<1 1 1>
    componentLinearVelScale=<1 1 1>
    gravityOverride {z=-980}
    physicsBodyDefinitions [{
      boundBone {boneName=dyn_bag_02}
      boxExtents=<7 7 7>
      constraintSetup {angularLimitsMax=<5 0 10>, angularLimitsMin=<-5 -25 -20>, angularTarget=<1 0 0>}
      localJointOffset=<3 0 0>
     }]
   }
   n5 = Anim Dynamics - Bone: dyn_bag_04(ComponentPose=@n4, Alpha=1, ExternalForce="0,0,0") {
    angularDampingOverride=1
    angularSpringConstant=500
    bAngularSpring=true
    bOverrideAngularDamping=true
    bOverrideLinearDamping=true
    bUseGravityOverride=true
    linearDampingOverride=1
    simulationSpace=RootRelative
    boundBone {boneName=dyn_bag_04}
    componentAppliedLinearAccClamp=<2000 2000 2000>
    componentLinearAccScale=<1 1 1>
    componentLinearVelScale=<1 1 1>
    gravityOverride {z=-980}
    physicsBodyDefinitions [{
      boundBone {boneName=dyn_bag_04}
      boxExtents=<7 7 7>
      constraintSetup {angularLimitsMax=<5 0 10>, angularLimitsMin=<-5 -25 -20>, angularTarget=<1 0 0>}
      localJointOffset=<3 0 0>
     }]
   }
   n6 = Anim Dynamics - Bone: dyn_bag_06(ComponentPose=@n5, Alpha=1, ExternalForce="0,0,0") {
    angularDampingOverride=1
    angularSpringConstant=500
    bAngularSpring=true
    bOverrideAngularDamping=true
    bOverrideLinearDamping=true
    bUseGravityOverride=true
    linearDampingOverride=1
    simulationSpace=RootRelative
    boundBone {boneName=dyn_bag_06}
    componentAppliedLinearAccClamp=<2000 2000 2000>
    componentLinearAccScale=<1 1 1>
    componentLinearVelScale=<1 1 1>
    gravityOverride {z=-980}
    physicsBodyDefinitions [{
      boundBone {boneName=dyn_bag_06}
      boxExtents=<7 7 7>
      constraintSetup {angularLimitsMax=<5 0 10>, angularLimitsMin=<-5 -25 -20>, angularTarget=<1 0 0>}
      localJointOffset=<3 0 0>
     }]
   }
   Output Pose(Result=Component To Local(ComponentPose=@n6)) {layerGroup=None}
   n7 = Control Rig!(Source=Copy Pose From Mesh() {bUseAttachedParent=true}, Alpha, Is Playing Emote:bool=Get bIsPlayingEmote()) -> (Pose) {controlRigAssetReference {controlRigAsset=&/F_MED_WaningCacti_CtrlRig}}
   n8 = RigidBody(ComponentPose=Local To Component(LocalPose=@n7)) {
    bClampLinearTranslationLimitToRefPose=true
    bOverrideWorldGravity=true
    overridePhysicsAsset=&/F_MED_WaningCacti_Physics
    simulationSpace=BaseBoneSpace
    baseBoneRef {boneName=pelvis}
    componentAppliedLinearAccClamp=<5000 5000 5000>
    overrideWorldGravity {z=-980}
    simSpaceSettings {dampingAlpha=0, maxAngularAcceleration=15, maxAngularVelocity=3, maxLinearAcceleration=1100, maxLinearVelocity=280, worldAlpha=1}
   }
   Save cached pose 'toAnimDyn1'(Pose=Component To Local(ComponentPose="Transform (Modify) Bone - Bone: dyn_belt_anchor_l_02"(ComponentPose=@n8, Alpha=1, Translation="-1,0,0", Rotation="0,0,0", Scale="1,1,1") {translationMode=BMM_Additive, boneToModify {boneName=dyn_belt_anchor_l_02}, translation {x=-1}})) {cacheName=toAnimDyn1}
   n9 = Anim Dynamics - Bone: dyn_necklace_01(ComponentPose=Local To Component(LocalPose=Use cached pose 'toAnimDyn1'()), Alpha=1, ExternalForce="0,0,0") {
    angularDampingOverride=1
    angularSpringConstant=500
    bAngularSpring=true
    bOverrideAngularDamping=true
    bOverrideLinearDamping=true
    bUseGravityOverride=true
    linearDampingOverride=1
    simulationSpace=RootRelative
    boundBone {boneName=dyn_necklace_01}
    componentAppliedLinearAccClamp=<2000 2000 2000>
    componentLinearAccScale=<1 1 1>
    componentLinearVelScale=<1 1 1>
    gravityOverride {z=-980}
    physicsBodyDefinitions [{
      boundBone {boneName=dyn_necklace_01}
      boxExtents=<8 8 8>
      constraintSetup {angularLimitsMax=<5 0 5>, angularLimitsMin=<-5 -40 -5>, angularTarget=<1 0 0>}
      localJointOffset=<2 0 0>
     }]
   }
   n10 = Anim Dynamics - Bone: dyn_necklace_02(ComponentPose=@n9, Alpha=1, ExternalForce="0,0,0") {
    angularDampingOverride=1
    angularSpringConstant=500
    bAngularSpring=true
    bOverrideAngularDamping=true
    bOverrideLinearDamping=true
    bUseGravityOverride=true
    linearDampingOverride=1
    simulationSpace=RootRelative
    boundBone {boneName=dyn_necklace_02}
    componentAppliedLinearAccClamp=<2000 2000 2000>
    componentLinearAccScale=<1 1 1>
    componentLinearVelScale=<1 1 1>
    gravityOverride {z=-980}
    physicsBodyDefinitions [{
      boundBone {boneName=dyn_necklace_02}
      boxExtents=<4 4 4>
      constraintSetup {angularLimitsMax=<5 0 10>, angularLimitsMin=<-5 -20 -10>, angularTarget=<1 0 0>}
      localJointOffset=<2 0 0>
     }]
   }
   n11 = Anim Dynamics - Bone: dyn_torso_zippo_01(ComponentPose=@n10, Alpha=1, ExternalForce="0,0,0") {
    angularDampingOverride=1
    angularSpringConstant=500
    bAngularSpring=true
    bOverrideAngularDamping=true
    bOverrideLinearDamping=true
    bUseGravityOverride=true
    linearDampingOverride=1
    simulationSpace=RootRelative
    boundBone {boneName=dyn_torso_zippo_01}
    componentAppliedLinearAccClamp=<2000 2000 2000>
    componentLinearAccScale=<1 1 1>
    componentLinearVelScale=<1 1 1>
    gravityOverride {z=-980}
    physicsBodyDefinitions [{
      boundBone {boneName=dyn_torso_zippo_01}
      boxExtents=<4 4 4>
      constraintSetup {angularLimitsMax=<5 40 15>, angularLimitsMin=<-5 0 -15>, angularTarget=<1 0 0>}
      localJointOffset=<2 0 0>
     }]
   }
   n12 = Anim Dynamics - Bone: dyn_skirt_zippo_l_01(ComponentPose=@n11, Alpha=1, ExternalForce="0,0,0") {
    angularDampingOverride=1
    angularSpringConstant=500
    bAngularSpring=true
    bOverrideAngularDamping=true
    bOverrideLinearDamping=true
    bUseGravityOverride=true
    linearDampingOverride=1
    simulationSpace=RootRelative
    boundBone {boneName=dyn_skirt_zippo_l_01}
    componentAppliedLinearAccClamp=<2000 2000 2000>
    componentLinearAccScale=<1 1 1>
    componentLinearVelScale=<1 1 1>
    gravityOverride {z=-980}
    physicsBodyDefinitions [{
      boundBone {boneName=dyn_skirt_zippo_l_01}
      boxExtents=<4 4 4>
      constraintSetup {angularLimitsMax=<5 5 15>, angularLimitsMin=<-5 -35 -15>, angularTarget=<1 0 0>}
      localJointOffset=<2 0 0>
     }]
   }
   n13 = Anim Dynamics - Bone: dyn_skirt_zippo_r_01(ComponentPose=@n12, Alpha=1, ExternalForce="0,0,0") {
    angularDampingOverride=1
    angularSpringConstant=500
    bAngularSpring=true
    bOverrideAngularDamping=true
    bOverrideLinearDamping=true
    bUseGravityOverride=true
    linearDampingOverride=1
    simulationSpace=RootRelative
    boundBone {boneName=dyn_skirt_zippo_r_01}
    componentAppliedLinearAccClamp=<2000 2000 2000>
    componentLinearAccScale=<1 1 1>
    componentLinearVelScale=<1 1 1>
    gravityOverride {z=-980}
    physicsBodyDefinitions [{
      boundBone {boneName=dyn_skirt_zippo_r_01}
      boxExtents=<4 4 4>
      constraintSetup {angularLimitsMax=<5 40 15>, angularLimitsMin=<-5 0 -15>, angularTarget=<1 0 0>}
      localJointOffset=<2 0 0>
     }]
   }
   n14 = Anim Dynamics - Bone: dyn_skirt_zippo_mid_01(ComponentPose=@n13, Alpha=1, ExternalForce="0,0,0") {
    angularDampingOverride=1
    angularSpringConstant=500
    bAngularSpring=true
    bOverrideAngularDamping=true
    bOverrideLinearDamping=true
    bUseGravityOverride=true
    linearDampingOverride=1
    simulationSpace=RootRelative
    boundBone {boneName=dyn_skirt_zippo_mid_01}
    componentAppliedLinearAccClamp=<2000 2000 2000>
    componentLinearAccScale=<1 1 1>
    componentLinearVelScale=<1 1 1>
    gravityOverride {z=-980}
    physicsBodyDefinitions [{
      boundBone {boneName=dyn_skirt_zippo_mid_01}
      boxExtents=<4 4 4>
      constraintSetup {angularLimitsMax=<5 5 15>, angularLimitsMin=<-5 -40 -15>, angularTarget=<1 0 0>}
      localJointOffset=<2 0 0>
     }]
   }
   Save cached pose 'toAnimDyn2'(Pose=Component To Local(ComponentPose=@n14)) {cacheName=toAnimDyn2}
   n15 = Anim Dynamics - Bone: dyn_leg_lace_l_01(ComponentPose=Local To Component(LocalPose=Use cached pose 'toAnimDyn2'()), Alpha=1, ExternalForce="0,0,0") {
    angularDampingOverride=1
    angularSpringConstant=500
    bAngularSpring=true
    bOverrideAngularDamping=true
    bOverrideLinearDamping=true
    bUseGravityOverride=true
    linearDampingOverride=1
    simulationSpace=RootRelative
    boundBone {boneName=dyn_leg_lace_l_01}
    componentAppliedLinearAccClamp=<2000 2000 2000>
    componentLinearAccScale=<1 1 1>
    componentLinearVelScale=<1 1 1>
    gravityOverride {z=-980}
    physicsBodyDefinitions [{
      boundBone {boneName=dyn_leg_lace_l_01}
      boxExtents=<4 4 4>
      constraintSetup {angularLimitsMax=<5 10 40>, angularLimitsMin=<-5 -20 -5>, angularTarget=<1 0 0>}
      localJointOffset=<3 0 0>
     }]
   }
   n16 = Anim Dynamics - Bone: dyn_leg_lace_l_03(ComponentPose=@n15, Alpha=1, ExternalForce="0,0,0") {
    angularDampingOverride=1
    angularSpringConstant=500
    bAngularSpring=true
    bOverrideAngularDamping=true
    bOverrideLinearDamping=true
    bUseGravityOverride=true
    linearDampingOverride=1
    simulationSpace=RootRelative
    boundBone {boneName=dyn_leg_lace_l_03}
    componentAppliedLinearAccClamp=<2000 2000 2000>
    componentLinearAccScale=<1 1 1>
    componentLinearVelScale=<1 1 1>
    gravityOverride {z=-980}
    physicsBodyDefinitions [{
      boundBone {boneName=dyn_leg_lace_l_03}
      boxExtents=<4 4 4>
      constraintSetup {angularLimitsMax=<5 10 40>, angularLimitsMin=<-5 -20 -5>, angularTarget=<1 0 0>}
      localJointOffset=<3 0 0>
     }]
   }
   n17 = Anim Dynamics - Bone: dyn_leg_lace_l_02(ComponentPose=@n16, Alpha=1, ExternalForce="0,0,0") {
    angularDampingOverride=1
    angularSpringConstant=500
    bAngularSpring=true
    bOverrideAngularDamping=true
    bOverrideLinearDamping=true
    bUseGravityOverride=true
    linearDampingOverride=1
    simulationSpace=RootRelative
    boundBone {boneName=dyn_leg_lace_l_02}
    componentAppliedLinearAccClamp=<2000 2000 2000>
    componentLinearAccScale=<1 1 1>
    componentLinearVelScale=<1 1 1>
    gravityOverride {z=-980}
    physicsBodyDefinitions [{
      boundBone {boneName=dyn_leg_lace_l_02}
      boxExtents=<4 4 4>
      constraintSetup {angularLimitsMax=<5 20 40>, angularLimitsMin=<-5 -10 -5>, angularTarget=<1 0 0>}
      localJointOffset=<3 0 0>
     }]
   }
   n18 = Anim Dynamics - Bone: dyn_leg_lace_l_04(ComponentPose=@n17, Alpha=1, ExternalForce="0,0,0") {
    angularDampingOverride=1
    angularSpringConstant=500
    bAngularSpring=true
    bOverrideAngularDamping=true
    bOverrideLinearDamping=true
    bUseGravityOverride=true
    linearDampingOverride=1
    simulationSpace=RootRelative
    boundBone {boneName=dyn_leg_lace_l_04}
    componentAppliedLinearAccClamp=<2000 2000 2000>
    componentLinearAccScale=<1 1 1>
    componentLinearVelScale=<1 1 1>
    gravityOverride {z=-980}
    physicsBodyDefinitions [{
      boundBone {boneName=dyn_leg_lace_l_04}
      boxExtents=<4 4 4>
      constraintSetup {angularLimitsMax=<5 20 40>, angularLimitsMin=<-5 -10 -5>, angularTarget=<1 0 0>}
      localJointOffset=<3 0 0>
     }]
   }
   n19 = Anim Dynamics - Bone: dyn_leg_lace_r_02(ComponentPose=@n18, Alpha=1, ExternalForce="0,0,0") {
    angularDampingOverride=1
    angularSpringConstant=500
    bAngularSpring=true
    bOverrideAngularDamping=true
    bOverrideLinearDamping=true
    bUseGravityOverride=true
    linearDampingOverride=1
    simulationSpace=RootRelative
    boundBone {boneName=dyn_leg_lace_r_02}
    componentAppliedLinearAccClamp=<2000 2000 2000>
    componentLinearAccScale=<1 1 1>
    componentLinearVelScale=<1 1 1>
    gravityOverride {z=-980}
    physicsBodyDefinitions [{
      boundBone {boneName=dyn_leg_lace_r_02}
      boxExtents=<4 4 4>
      constraintSetup {angularLimitsMax=<10 5 40>, angularLimitsMin=<-10 -5 0>, angularTarget=<1 0 0>}
      localJointOffset=<-3 0 0>
     }]
   }
   n20 = Anim Dynamics - Bone: dyn_leg_lace_r_04(ComponentPose=@n19, Alpha=1, ExternalForce="0,0,0") {
    angularDampingOverride=1
    angularSpringConstant=500
    bAngularSpring=true
    bOverrideAngularDamping=true
    bOverrideLinearDamping=true
    bUseGravityOverride=true
    linearDampingOverride=1
    simulationSpace=RootRelative
    boundBone {boneName=dyn_leg_lace_r_04}
    componentAppliedLinearAccClamp=<2000 2000 2000>
    componentLinearAccScale=<1 1 1>
    componentLinearVelScale=<1 1 1>
    gravityOverride {z=-980}
    physicsBodyDefinitions [{
      boundBone {boneName=dyn_leg_lace_r_04}
      boxExtents=<4 4 4>
      constraintSetup {angularLimitsMax=<5 10 45>, angularLimitsMin=<-5 -20 -5>, angularTarget=<1 0 0>}
      localJointOffset=<-3 0 0>
     }]
   }
   n21 = Anim Dynamics - Bone: dyn_leg_lace_r_01(ComponentPose=@n20, Alpha=1, ExternalForce="0,0,0") {
    angularDampingOverride=1
    angularSpringConstant=500
    bAngularSpring=true
    bOverrideAngularDamping=true
    bOverrideLinearDamping=true
    bUseGravityOverride=true
    linearDampingOverride=1
    simulationSpace=RootRelative
    boundBone {boneName=dyn_leg_lace_r_01}
    componentAppliedLinearAccClamp=<2000 2000 2000>
    componentLinearAccScale=<1 1 1>
    componentLinearVelScale=<1 1 1>
    gravityOverride {z=-980}
    physicsBodyDefinitions [{
      boundBone {boneName=dyn_leg_lace_r_01}
      boxExtents=<4 4 4>
      constraintSetup {angularLimitsMax=<5 20 40>, angularLimitsMin=<-5 -10 -5>, angularTarget=<1 0 0>}
      localJointOffset=<-3 0 0>
     }]
   }
   n22 = Anim Dynamics - Bone: dyn_leg_lace_r_03(ComponentPose=@n21, Alpha=1, ExternalForce="0,0,0") {
    angularDampingOverride=1
    angularSpringConstant=500
    bAngularSpring=true
    bOverrideAngularDamping=true
    bOverrideLinearDamping=true
    bUseGravityOverride=true
    linearDampingOverride=1
    simulationSpace=RootRelative
    boundBone {boneName=dyn_leg_lace_r_03}
    componentAppliedLinearAccClamp=<2000 2000 2000>
    componentLinearAccScale=<1 1 1>
    componentLinearVelScale=<1 1 1>
    gravityOverride {z=-980}
    physicsBodyDefinitions [{
      boundBone {boneName=dyn_leg_lace_r_03}
      boxExtents=<4 4 4>
      constraintSetup {angularTargetAxis=AxisY, angularLimitsMax=<5 20 45>, angularLimitsMin=<-5 -10 -5>, angularTarget=<1 0 0>}
      localJointOffset=<-3 0 0>
     }]
   }
   Save cached pose 'toAnimDyn3'(Pose=Component To Local(ComponentPose=@n22)) {cacheName=toAnimDyn3}
   n23 = Anim Dynamics - Bone: dyn_skirt_bk_mid_01(ComponentPose=Local To Component(LocalPose=Use cached pose 'toAnimDyn3'()), Alpha=1, ExternalForce="0,0,0") {
    angularDampingOverride=1
    angularSpringConstant=500
    bAngularSpring=true
    bOverrideAngularDamping=true
    bOverrideLinearDamping=true
    bUseGravityOverride=true
    linearDampingOverride=1
    simulationSpace=RootRelative
    boundBone {boneName=dyn_skirt_bk_mid_01}
    componentAppliedLinearAccClamp=<5000 5000 5000>
    componentLinearAccScale=<1 1 1>
    componentLinearVelScale=<1 1 1>
    gravityOverride {z=-980}
    physicsBodyDefinitions [{
      boundBone {boneName=dyn_skirt_bk_mid_01}
      boxExtents=<10 10 10>
      constraintSetup {angularLimitsMax=<0 10 0>, angularLimitsMin=<0 -20 0>, angularTarget=<1 0 0>}
      localJointOffset=<1 0 0>
     }]
   }
   n24 = Anim Dynamics - Bone: dyn_skirt_bk_r_base(ComponentPose=@n23, Alpha=1, ExternalForce="0,0,0") {
    angularDampingOverride=1
    angularSpringConstant=500
    bAngularSpring=true
    bOverrideAngularDamping=true
    bOverrideLinearDamping=true
    bUseGravityOverride=true
    linearDampingOverride=1
    simulationSpace=RootRelative
    boundBone {boneName=dyn_skirt_bk_r_base}
    componentAppliedLinearAccClamp=<5000 5000 5000>
    componentLinearAccScale=<1 1 1>
    componentLinearVelScale=<1 1 1>
    gravityOverride {z=-980}
    physicsBodyDefinitions [{
      boundBone {boneName=dyn_skirt_bk_r_base}
      boxExtents=<10 10 10>
      constraintSetup {angularLimitsMax=<0 20 0>, angularLimitsMin=<0 0 0>, angularTarget=<1 0 0>}
      localJointOffset=<1 0 0>
     }]
   }
   n25 = Anim Dynamics - Bone: dyn_skirt_bk_l_base(ComponentPose=@n24, Alpha=1, ExternalForce="0,0,0") {
    angularDampingOverride=1
    angularSpringConstant=500
    bAngularSpring=true
    bOverrideAngularDamping=true
    bOverrideLinearDamping=true
    bUseGravityOverride=true
    linearDampingOverride=1
    simulationSpace=RootRelative
    boundBone {boneName=dyn_skirt_bk_l_base}
    componentAppliedLinearAccClamp=<5000 5000 5000>
    componentLinearAccScale=<1 1 1>
    componentLinearVelScale=<1 1 1>
    gravityOverride {z=-980}
    physicsBodyDefinitions [{
      boundBone {boneName=dyn_skirt_bk_l_base}
      boxExtents=<10 10 10>
      constraintSetup {angularLimitsMax=<0 0 0>, angularLimitsMin=<0 -20 0>, angularTarget=<1 0 0>}
      localJointOffset=<1 0 0>
     }]
   }
   n26 = Anim Dynamics - Bone: dyn_skirt_frnt_mid_01(ComponentPose=@n25, Alpha=1, ExternalForce="0,0,0") {
    angularDampingOverride=1
    angularSpringConstant=500
    bAngularSpring=true
    bOverrideAngularDamping=true
    bOverrideLinearDamping=true
    bUseGravityOverride=true
    linearDampingOverride=1
    simulationSpace=RootRelative
    boundBone {boneName=dyn_skirt_frnt_mid_01}
    componentAppliedLinearAccClamp=<5000 5000 5000>
    componentLinearAccScale=<1 1 1>
    componentLinearVelScale=<1 1 1>
    gravityOverride {z=-980}
    physicsBodyDefinitions [{
      boundBone {boneName=dyn_skirt_frnt_mid_01}
      boxExtents=<10 10 10>
      constraintSetup {angularLimitsMax=<0 0 0>, angularLimitsMin=<0 -25 0>, angularTarget=<1 0 0>}
      localJointOffset=<1 0 0>
     }]
   }
   n27 = Anim Dynamics - Bone: dyn_skirt_frnt_l_01(ComponentPose=@n26, Alpha=1, ExternalForce="0,0,0") {
    angularDampingOverride=1
    angularSpringConstant=500
    bAngularSpring=true
    bOverrideAngularDamping=true
    bOverrideLinearDamping=true
    bUseGravityOverride=true
    linearDampingOverride=1
    simulationSpace=RootRelative
    boundBone {boneName=dyn_skirt_frnt_l_01}
    componentAppliedLinearAccClamp=<5000 5000 5000>
    componentLinearAccScale=<1 1 1>
    componentLinearVelScale=<1 1 1>
    gravityOverride {z=-980}
    physicsBodyDefinitions [{
      boundBone {boneName=dyn_skirt_frnt_l_01}
      boxExtents=<10 10 10>
      constraintSetup {angularLimitsMax=<0 0 0>, angularLimitsMin=<0 -25 0>, angularTarget=<1 0 0>}
      localJointOffset=<1 0 0>
     }]
   }
   n28 = Anim Dynamics - Bone: dyn_skirt_frnt_r_01(ComponentPose=@n27, Alpha=1, ExternalForce="0,0,0") {
    angularDampingOverride=1
    angularSpringConstant=500
    bAngularSpring=true
    bOverrideAngularDamping=true
    bOverrideLinearDamping=true
    bUseGravityOverride=true
    linearDampingOverride=1
    simulationSpace=RootRelative
    boundBone {boneName=dyn_skirt_frnt_r_01}
    componentAppliedLinearAccClamp=<5000 5000 5000>
    componentLinearAccScale=<1 1 1>
    componentLinearVelScale=<1 1 1>
    gravityOverride {z=-980}
    physicsBodyDefinitions [{
      boundBone {boneName=dyn_skirt_frnt_r_01}
      boxExtents=<10 10 10>
      constraintSetup {angularLimitsMax=<0 25 0>, angularLimitsMin=<0 0 0>, angularTarget=<1 0 0>}
      localJointOffset=<1 0 0>
     }]
   }
   n29 = Anim Dynamics - Bone: dyn_skirt_side_l_01(ComponentPose=@n28, Alpha=1, ExternalForce="0,0,0") {
    angularDampingOverride=1
    angularSpringConstant=500
    bAngularSpring=true
    bOverrideAngularDamping=true
    bOverrideLinearDamping=true
    bUseGravityOverride=true
    linearDampingOverride=1
    simulationSpace=RootRelative
    boundBone {boneName=dyn_skirt_side_l_01}
    componentAppliedLinearAccClamp=<5000 5000 5000>
    componentLinearAccScale=<1 1 1>
    componentLinearVelScale=<1 1 1>
    gravityOverride {z=-980}
    physicsBodyDefinitions [{
      boundBone {boneName=dyn_skirt_side_l_01}
      boxExtents=<10 10 10>
      constraintSetup {angularLimitsMax=<0 0 0>, angularLimitsMin=<0 -25 0>, angularTarget=<1 0 0>}
      localJointOffset=<1 0 0>
     }]
   }
   n30 = Anim Dynamics - Bone: dyn_skirt_side_r_01(ComponentPose=@n29, Alpha=1, ExternalForce="0,0,0") {
    angularDampingOverride=1
    angularSpringConstant=500
    bAngularSpring=true
    bOverrideAngularDamping=true
    bOverrideLinearDamping=true
    bUseGravityOverride=true
    linearDampingOverride=1
    simulationSpace=RootRelative
    boundBone {boneName=dyn_skirt_side_r_01}
    componentAppliedLinearAccClamp=<5000 5000 5000>
    componentLinearAccScale=<1 1 1>
    componentLinearVelScale=<1 1 1>
    gravityOverride {z=-980}
    physicsBodyDefinitions [{
      boundBone {boneName=dyn_skirt_side_r_01}
      boxExtents=<10 10 10>
      constraintSetup {angularLimitsMax=<0 25 0>, angularLimitsMin=<0 0 0>, angularTarget=<1 0 0>}
      localJointOffset=<1 0 0>
     }]
   }
   Save cached pose 'toConstraints'(Pose=Component To Local(ComponentPose=@n30)) {cacheName=toConstraints}
   n31 = "Constraint - dyn_skirt_zippo_l_base (PPPPPPPP)"!(
    ComponentPose=Local To Component(LocalPose=Use cached pose 'toConstraints'())
    Alpha=1
    ConstraintWeights_0=0.034
    ConstraintWeights_1:float=0.0781
    ConstraintWeights_2:float=0.0014
    ConstraintWeights_3:float=0.0228
    ConstraintWeights_4:float=0.0387
    ConstraintWeights_5:float=0.0003
    ConstraintWeights_6:float=0.7242
    ConstraintWeights_7:float=0.1004
   ) -> (Pose) {
    boneToModify {boneName=dyn_skirt_zippo_l_base}
    constraintSetup [
     {targetBone {boneName=pelvis}}
     {targetBone {boneName=thigh_twist_01_l}}
     {targetBone {boneName=thigh_twist_01_r}}
     {targetBone {boneName=thigh_l}}
     {targetBone {boneName=deform_groin_l}}
     {targetBone {boneName=deform_groin_r}}
     {targetBone {boneName=dyn_skirt_frnt_l_01}}
     {targetBone {boneName=dyn_skirt_frnt_mid_01}}
    ]
    constraintWeights [1,1,1,1,1,1,1,1]
   }
   n32 = "Constraint - dyn_skirt_zippo_r_base (PPPPPPPP)"!(
    ComponentPose=@n31
    Alpha=1
    ConstraintWeights_0=0.04
    ConstraintWeights_1:float=0.0015
    ConstraintWeights_2:float=0.4258
    ConstraintWeights_3:float=0.0454
    ConstraintWeights_4:float=0.0006
    ConstraintWeights_5:float=0.0272
    ConstraintWeights_6:float=0.3678
    ConstraintWeights_7:float=0.0917
   ) -> (Pose) {
    boneToModify {boneName=dyn_skirt_zippo_r_base}
    constraintSetup [
     {targetBone {boneName=pelvis}}
     {targetBone {boneName=thigh_twist_01_l}}
     {targetBone {boneName=thigh_twist_01_r}}
     {targetBone {boneName=thigh_r}}
     {targetBone {boneName=deform_groin_l}}
     {targetBone {boneName=deform_groin_r}}
     {targetBone {boneName=dyn_skirt_frnt_r_01}}
     {targetBone {boneName=dyn_skirt_frnt_mid_01}}
    ]
    constraintWeights [1,1,1,1,1,1,1,1]
   }
   n33 = "Constraint - dyn_skirt_zippo_mid_base (PPPPPPPPP)"!(
    ComponentPose=@n32
    Alpha=1
    ConstraintWeights_0=0.0854
    ConstraintWeights_1:float=0.0267
    ConstraintWeights_2:float=0.0237
    ConstraintWeights_3:float=0.0019
    ConstraintWeights_4:float=0.0254
    ConstraintWeights_5:float=0.0185
    ConstraintWeights_6:float=0.0132
    ConstraintWeights_7:float=0.035
    ConstraintWeights_8:float=0.77
   ) -> (Pose) {
    boneToModify {boneName=dyn_skirt_zippo_mid_base}
    constraintSetup [
     {targetBone {boneName=pelvis}}
     {targetBone {boneName=thigh_twist_01_l}}
     {targetBone {boneName=thigh_twist_01_r}}
     {targetBone {boneName=thigh_r}}
     {targetBone {boneName=deform_groin_l}}
     {targetBone {boneName=deform_groin_r}}
     {targetBone {boneName=dyn_skirt_frnt_l_01}}
     {targetBone {boneName=dyn_skirt_frnt_r_01}}
     {targetBone {boneName=dyn_skirt_frnt_mid_01}}
    ]
    constraintWeights [1,1,1,1,1,1,1,1,1]
   }
   n34 = "Constraint - dyn_torso_zippo_base (PPPPPPP)"!(
    ComponentPose=@n33
    Alpha=1
    ConstraintWeights_0=0.0452
    ConstraintWeights_1:float=0.8747
    ConstraintWeights_2:float=0.0398
    ConstraintWeights_3:float=0.0402
    ConstraintWeights_4:float=0
    ConstraintWeights_5:float=0
    ConstraintWeights_6:float=0
   ) -> (Pose) {
    boneToModify {boneName=dyn_torso_zippo_base}
    constraintSetup [
     {targetBone {boneName=spine_05}}
     {targetBone {boneName=spine_04}}
     {targetBone {boneName=spine_03}}
     {targetBone {boneName=spine_02}}
     {targetBone {boneName=spine_01}}
     {targetBone {boneName=deform_pec_l}}
     {targetBone {boneName=deform_pec_r}}
    ]
    constraintWeights [1,1,1,1,1,1,1]
   }
   Save cached pose 'toAnimDyn4'(Pose=Component To Local(ComponentPose=@n34)) {cacheName=toAnimDyn4}
  ]}
 EventGraph {nodes [Event BlueprintUpdateAnimation()]}
]
variables [
 Gravity Override {type=Vector}
 RBAN Settings Out {type=FortRigidBodyAnimNodeParameters}
]
