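# Physics asset for the F_MED_WaningCacti face accessory mesh, covering dynamic hair chains plus upper-body capsules for spine, neck, head, clavicles and upper arms.
# Seven back and side hair chains hang from zero-size base bodies constrained to the head, with angle-limited swing and twist down each chain.
# Side-left and back-side-right chains branch from their first segment.
# Angularly free cross-constraints link the left and right back chains to the middle one at segment 04.
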PhysicsAsset /BRCosmetics/Characters/Player/Female/Medium/Bodies/F_MED_WaningCacti/Meshes/Parts/F_MED_WaningCacti_FaceAcc_Physics
bodies {
 clavicle_l [{shapeType=Capsule, length=5.4102, radius=4.8409, shapeName=clavicle_l_capsule, center=<5.0609 -1.2375 0>, rotation=<rot 90 0.0002 0>}]
 clavicle_r [{shapeType=Capsule, length=5.8392, radius=5.8299, shapeName=clavicle_r_capsule, center=<-5.0609 1.2375 0>, rotation=<rot 90 180 0>}]
 dyn_hair_bk_l_01 [{shapeType=Capsule, length=6.5613, radius=3.4387, shapeName=dyn_hair_bk_l_01_capsule, center=<3.3687 0 0>}]
 dyn_hair_bk_l_02 [{shapeType=Capsule, length=6, radius=4, shapeName=dyn_hair_bk_l_02_capsule, center=<3.3687 0 0>}]
 dyn_hair_bk_l_03 [{shapeType=Capsule, length=6, radius=4, shapeName=dyn_hair_bk_l_03_capsule, center=<3.3687 0 0>, rotation=<rot 0 0 90>}]
 dyn_hair_bk_l_04 [{shapeType=Capsule, length=6, radius=4, shapeName=dyn_hair_bk_l_04_capsule, center=<3.3687 0 0>, rotation=<rot 0 0 90>}]
 dyn_hair_bk_l_05 [{shapeType=Capsule, length=5.2739, radius=4.7261, shapeName=dyn_hair_bk_l_05_capsule, center=<3.3687 0 0>, rotation=<rot 0 0 90>}]
 dyn_hair_bk_l_base [{shapeType=Capsule, length=0, radius=0, shapeName=dyn_hair_bk_l_base_capsule}]
 dyn_hair_bk_mid_01 [{shapeType=Capsule, length=6.5613, radius=3.4387, shapeName=dyn_hair_bk_mid_01_capsule, center=<3.3687 0 0>}]
 dyn_hair_bk_mid_02 [{shapeType=Capsule, length=6, radius=4, shapeName=dyn_hair_bk_mid_02_capsule, center=<3.3687 0 0>}]
 dyn_hair_bk_mid_03 [{shapeType=Capsule, length=6, radius=4, shapeName=dyn_hair_bk_mid_03_capsule, center=<3.3687 0 0>, rotation=<rot 0 0 90>}]
 dyn_hair_bk_mid_04 [{shapeType=Capsule, length=6, radius=4, shapeName=dyn_hair_bk_mid_04_capsule, center=<3.3687 0 0>, rotation=<rot 0 0 90>}]
 dyn_hair_bk_mid_05 [{shapeType=Capsule, length=5.2739, radius=4.7261, shapeName=dyn_hair_bk_mid_05_capsule, center=<3.3687 0 0>, rotation=<rot 0 0 90>}]
 dyn_hair_bk_mid_base [{shapeType=Capsule, length=0, radius=0, shapeName=dyn_hair_bk_mid_base_capsule}]
 dyn_hair_bk_r_01 [{shapeType=Capsule, length=6.5613, radius=3.4387, shapeName=dyn_hair_bk_r_01_capsule, center=<3.3687 0 0>}]
 dyn_hair_bk_r_02 [{shapeType=Capsule, length=5.4921, radius=4.5079, shapeName=dyn_hair_bk_r_02_capsule, center=<3.3687 0 0>}]
 dyn_hair_bk_r_03 [{shapeType=Capsule, length=6, radius=4, shapeName=dyn_hair_bk_r_03_capsule, center=<3.3687 0 0>, rotation=<rot 0 0 90>}]
 dyn_hair_bk_r_04 [{shapeType=Capsule, length=6, radius=4, shapeName=dyn_hair_bk_r_04_capsule, center=<3.3687 0 0>, rotation=<rot 0 0 90>}]
 dyn_hair_bk_r_05 [{shapeType=Capsule, length=5.2739, radius=4.7261, shapeName=dyn_hair_bk_r_05_capsule, center=<3.3687 0 0>, rotation=<rot 0 0 90>}]
 dyn_hair_bk_r_base [{shapeType=Capsule, length=0, radius=0, shapeName=dyn_hair_bk_r_base_capsule}]
 dyn_hair_bk_side_l_01 [{shapeType=Capsule, length=4, radius=1.8506, shapeName=dyn_hair_bk_side_l_01_capsule, center=<3.2803 0 0>, rotation=<rot 90 0 0>}]
 dyn_hair_bk_side_l_02 [{shapeType=Capsule, length=4, radius=1.8506, shapeName=dyn_hair_bk_side_l_02_capsule, center=<3.2803 0 0>, rotation=<rot 90 0 0>}]
 dyn_hair_bk_side_l_base [{shapeType=Capsule, length=0, radius=0, shapeName=dyn_hair_bk_side_l_base_capsule}]
 dyn_hair_bk_side_r_01 [{shapeType=Capsule, length=6.945, radius=1.555, shapeName=dyn_hair_bk_side_r_01_capsule, center=<3.0718 0 0>, rotation=<rot 90 0 0>}]
 dyn_hair_bk_side_r_02 [{shapeType=Capsule, length=6.945, radius=1.555, shapeName=dyn_hair_bk_side_r_02_capsule, center=<3.4252 0 0>, rotation=<rot 90 0 0>}]
 dyn_hair_bk_side_r_03 [{shapeType=Capsule, length=3.6739, radius=1.8276, shapeName=dyn_hair_bk_side_r_03_capsule, center=<3.31 0 0>, rotation=<rot 90 0 0>}]
 dyn_hair_bk_side_r_base [{shapeType=Capsule, length=0, radius=0, shapeName=dyn_hair_bk_side_r_base_capsule}]
 dyn_hair_side_l_01 [{shapeType=Capsule, length=9.6645, radius=2.3355, shapeName=dyn_hair_side_l_01_capsule, center=<3.9399 0 0>, rotation=<rot 0 0 90>}]
 dyn_hair_side_l_02 [{shapeType=Capsule, length=5.9592, radius=3.2682, shapeName=dyn_hair_side_l_02_capsule, center=<2.8442 0 0>}]
 dyn_hair_side_l_03 [{
   shapeType=Capsule
   length=6.4936
   radius=2.3004
   shapeName=dyn_hair_side_l_03_capsule
   center=<3.2258 -0.0001 0.4922>
   rotation=<rot 80.2977 -179.9323 179.9999>
  }]
 dyn_hair_side_l_04 [{shapeType=Capsule, length=3.1062, radius=4, shapeName=dyn_hair_side_l_04_capsule, center=<2.8493 0 0>, rotation=<rot 90 0 0>}]
 dyn_hair_side_l_base [{shapeType=Capsule, length=0, radius=0, shapeName=dyn_hair_side_l_base_capsule}]
 dyn_hair_side_r_01 [{shapeType=Capsule, length=9.4102, radius=2.5898, shapeName=dyn_hair_side_r_01_capsule, center=<3.0718 0 0>, rotation=<rot 0 0 90>}]
 dyn_hair_side_r_02 [{shapeType=Capsule, length=2.62, radius=2.9403, shapeName=dyn_hair_side_r_02_capsule, center=<3.4252 0 0>}]
 dyn_hair_side_r_03 [{shapeType=Capsule, length=1.8459, radius=3.7253, shapeName=dyn_hair_side_r_03_capsule, center=<3.31 0 0>}]
 dyn_hair_side_r_base [{shapeType=Capsule, length=0, radius=0, shapeName=dyn_hair_side_r_base_capsule}]
 head [{shapeType=Capsule, length=4.5634, radius=6.4988, shapeName=head_capsule, center=<4.5492 0.8963 0>, rotation=<rot 90 0 0>}]
 neck_01 [{shapeType=Capsule, length=8.5279, radius=3.6816, shapeName=neck_01_capsule, rotation=<rot 90 0 0>}]
 neck_02 [{shapeType=Capsule, length=8.3655, radius=3.844, shapeName=neck_02_capsule, rotation=<rot 90 0 0>}]
 spine_04 [{shapeType=Capsule, length=15.843, radius=10.4793, shapeName=spine_04_capsule, center=<3.8419 1.1899 0>}]
 spine_05 [{shapeType=Capsule, length=18.0449, radius=9.8306, shapeName=spine_05_capsule, center=<0.8555 2.8358 0>}]
 upperarm_l [{shapeType=Capsule, length=22.1831, radius=4.7147, shapeName=upperarm_l_capsule, center=<12.3805 0 0>, rotation=<rot 90 0 0>}]
 upperarm_r [{shapeType=Capsule, length=23.9019, radius=2.9749, shapeName=upperarm_r_capsule, center=<-12.3805 0 0>, rotation=<rot 90 180 0>}]
}
constraints [
 {
  bone1Name=spine_05
  bone2Name=spine_04
  swing1LimitDegrees=45
  swing1Motion=Free
  swing2LimitDegrees=45
  swing2Motion=Free
  twistLimitDegrees=45
  twistMotion=Free
 }
 {
  bone1Name=neck_01
  bone2Name=spine_05
  swing1LimitDegrees=45
  swing1Motion=Free
  swing2LimitDegrees=45
  swing2Motion=Free
  twistLimitDegrees=45
  twistMotion=Free
 }
 {
  bone1Name=neck_02
  bone2Name=neck_01
  swing1LimitDegrees=45
  swing1Motion=Free
  swing2LimitDegrees=45
  swing2Motion=Free
  twistLimitDegrees=45
  twistMotion=Free
 }
 {
  bone1Name=head
  bone2Name=neck_02
  swing1LimitDegrees=45
  swing1Motion=Free
  swing2LimitDegrees=45
  swing2Motion=Free
  twistLimitDegrees=45
  twistMotion=Free
 }
 {
  bone1Name=dyn_hair_bk_side_l_base
  bone2Name=head
  swing1LimitDegrees=45
  swing1Motion=Free
  swing2LimitDegrees=45
  swing2Motion=Free
  twistLimitDegrees=45
  twistMotion=Free
 }
 {bone1Name=dyn_hair_bk_side_l_01, bone2Name=dyn_hair_bk_side_l_base, swing1LimitDegrees=20, swing2LimitDegrees=10, twistLimitDegrees=10}
 {bone1Name=dyn_hair_bk_side_l_02, bone2Name=dyn_hair_bk_side_l_01, swing1LimitDegrees=25, swing2LimitDegrees=10, twistLimitDegrees=10}
 {
  bone1Name=dyn_hair_bk_side_r_base
  bone2Name=head
  swing1LimitDegrees=45
  swing1Motion=Free
  swing2LimitDegrees=45
  swing2Motion=Free
  twistLimitDegrees=45
  twistMotion=Free
 }
 {bone1Name=dyn_hair_bk_side_r_01, bone2Name=dyn_hair_bk_side_r_base, swing1LimitDegrees=20, swing2LimitDegrees=20, twistLimitDegrees=10}
 {bone1Name=dyn_hair_bk_side_r_02, bone2Name=dyn_hair_bk_side_r_01, swing1LimitDegrees=30, swing2LimitDegrees=20, twistLimitDegrees=10}
 {bone1Name=dyn_hair_bk_side_r_03, bone2Name=dyn_hair_bk_side_r_01, swing1LimitDegrees=30, swing2LimitDegrees=30, twistLimitDegrees=10}
 {
  bone1Name=dyn_hair_side_l_base
  bone2Name=head
  swing1LimitDegrees=45
  swing1Motion=Free
  swing2LimitDegrees=45
  swing2Motion=Free
  twistLimitDegrees=45
  twistMotion=Free
 }
 {bone1Name=dyn_hair_side_l_01, bone2Name=dyn_hair_side_l_base, swing1LimitDegrees=20, swing2LimitDegrees=5, twistLimitDegrees=10}
 {bone1Name=dyn_hair_side_l_02, bone2Name=dyn_hair_side_l_01, swing1LimitDegrees=20, swing2LimitDegrees=20, twistLimitDegrees=10}
 {bone1Name=dyn_hair_side_l_03, bone2Name=dyn_hair_side_l_02, swing1LimitDegrees=20, swing2LimitDegrees=30, twistLimitDegrees=10}
 {
  bone1Name=dyn_hair_side_r_base
  bone2Name=head
  swing1LimitDegrees=45
  swing1Motion=Free
  swing2LimitDegrees=45
  swing2Motion=Free
  twistLimitDegrees=45
  twistMotion=Free
 }
 {bone1Name=dyn_hair_side_r_01, bone2Name=dyn_hair_side_r_base, swing1LimitDegrees=10, swing2LimitDegrees=5, twistLimitDegrees=10}
 {bone1Name=dyn_hair_side_r_02, bone2Name=dyn_hair_side_r_01, swing1LimitDegrees=5, swing2LimitDegrees=20, twistLimitDegrees=10}
 {bone1Name=dyn_hair_side_r_03, bone2Name=dyn_hair_side_r_02, swing1LimitDegrees=5, swing2LimitDegrees=20, twistLimitDegrees=10}
 {bone1Name=dyn_hair_side_l_04, bone2Name=dyn_hair_side_l_01, swing1LimitDegrees=20, swing2LimitDegrees=10, twistLimitDegrees=10}
 {
  bone1Name=clavicle_l
  bone2Name=spine_05
  swing1LimitDegrees=45
  swing1Motion=Free
  swing2LimitDegrees=45
  swing2Motion=Free
  twistLimitDegrees=45
  twistMotion=Free
 }
 {
  bone1Name=clavicle_r
  bone2Name=spine_05
  swing1LimitDegrees=45
  swing1Motion=Free
  swing2LimitDegrees=45
  swing2Motion=Free
  twistLimitDegrees=45
  twistMotion=Free
 }
 {
  bone1Name=upperarm_l
  bone2Name=clavicle_l
  swing1LimitDegrees=45
  swing1Motion=Free
  swing2LimitDegrees=45
  swing2Motion=Free
  twistLimitDegrees=45
  twistMotion=Free
 }
 {
  bone1Name=upperarm_r
  bone2Name=clavicle_r
  swing1LimitDegrees=45
  swing1Motion=Free
  swing2LimitDegrees=45
  swing2Motion=Free
  twistLimitDegrees=45
  twistMotion=Free
 }
 {
  bone1Name=dyn_hair_bk_l_base
  bone2Name=head
  swing1LimitDegrees=45
  swing1Motion=Free
  swing2LimitDegrees=45
  swing2Motion=Free
  twistLimitDegrees=45
  twistMotion=Free
 }
 {bone1Name=dyn_hair_bk_l_01, bone2Name=dyn_hair_bk_l_base, swing1LimitDegrees=45, swing2LimitDegrees=10, twistLimitDegrees=10}
 {bone1Name=dyn_hair_bk_l_02, bone2Name=dyn_hair_bk_l_01, swing1LimitDegrees=30, swing2LimitDegrees=10, twistLimitDegrees=10}
 {bone1Name=dyn_hair_bk_l_03, bone2Name=dyn_hair_bk_l_02, swing1LimitDegrees=10, swing2LimitDegrees=30, twistLimitDegrees=10}
 {bone1Name=dyn_hair_bk_l_04, bone2Name=dyn_hair_bk_l_03, swing1LimitDegrees=10, swing2LimitDegrees=30, twistLimitDegrees=10}
 {bone1Name=dyn_hair_bk_l_05, bone2Name=dyn_hair_bk_l_04, swing1LimitDegrees=10, swing2LimitDegrees=30, twistLimitDegrees=10}
 {
  bone1Name=dyn_hair_bk_mid_base
  bone2Name=head
  swing1LimitDegrees=45
  swing1Motion=Free
  swing2LimitDegrees=45
  swing2Motion=Free
  twistLimitDegrees=45
  twistMotion=Free
 }
 {bone1Name=dyn_hair_bk_mid_01, bone2Name=dyn_hair_bk_mid_base, swing1LimitDegrees=45, swing2LimitDegrees=10, twistLimitDegrees=10}
 {bone1Name=dyn_hair_bk_mid_02, bone2Name=dyn_hair_bk_mid_01, swing1LimitDegrees=30, swing2LimitDegrees=10, twistLimitDegrees=10}
 {bone1Name=dyn_hair_bk_mid_03, bone2Name=dyn_hair_bk_mid_02, swing1LimitDegrees=10, swing2LimitDegrees=30, twistLimitDegrees=10}
 {bone1Name=dyn_hair_bk_mid_04, bone2Name=dyn_hair_bk_mid_03, swing1LimitDegrees=10, swing2LimitDegrees=30, twistLimitDegrees=10}
 {bone1Name=dyn_hair_bk_mid_05, bone2Name=dyn_hair_bk_mid_04, swing1LimitDegrees=10, swing2LimitDegrees=30, twistLimitDegrees=10}
 {
  bone1Name=dyn_hair_bk_r_base
  bone2Name=head
  swing1LimitDegrees=45
  swing1Motion=Free
  swing2LimitDegrees=45
  swing2Motion=Free
  twistLimitDegrees=45
  twistMotion=Free
 }
 {bone1Name=dyn_hair_bk_r_01, bone2Name=dyn_hair_bk_r_base, swing1LimitDegrees=45, swing2LimitDegrees=10, twistLimitDegrees=10}
 {bone1Name=dyn_hair_bk_r_02, bone2Name=dyn_hair_bk_r_01, swing1LimitDegrees=30, swing2LimitDegrees=10, twistLimitDegrees=10}
 {bone1Name=dyn_hair_bk_r_03, bone2Name=dyn_hair_bk_r_02, swing1LimitDegrees=10, swing2LimitDegrees=30, twistLimitDegrees=10}
 {bone1Name=dyn_hair_bk_r_04, bone2Name=dyn_hair_bk_r_03, swing1LimitDegrees=10, swing2LimitDegrees=30, twistLimitDegrees=10}
 {bone1Name=dyn_hair_bk_r_05, bone2Name=dyn_hair_bk_r_04, swing1LimitDegrees=10, swing2LimitDegrees=30, twistLimitDegrees=10}
 {
  bone1Name=dyn_hair_bk_r_04
  bone2Name=dyn_hair_bk_mid_04
  swing1LimitDegrees=45
  swing1Motion=Free
  swing2LimitDegrees=45
  swing2Motion=Free
  twistLimitDegrees=45
  twistMotion=Free
 }
 {
  bone1Name=dyn_hair_bk_l_04
  bone2Name=dyn_hair_bk_mid_04
  swing1LimitDegrees=45
  swing1Motion=Free
  swing2LimitDegrees=45
  swing2Motion=Free
  twistLimitDegrees=45
  twistMotion=Free
 }
]
